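# Stereo one-shot MetaSound preset of the shared boss VO source that plays the HeavyLamp character's death voice lines.
# It feeds three death waves to the parent graph with guaranteed playback, a fixed delay, fixed volume boost and no pitch variation.
# A HeavyLamp effect chain and control bus shape the voice, which also sends to the NPC voice sidechain bus.
# Concurrency allows one instance per owner and blocks new ones.
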MetaSoundSource /JethroLager/Assets/Audio/VO/HeavyLamp/MSS/MSS_IP_VO_HeavyLamp_Death
preset_parent=/Game/Sounds/Fort_Human_VO_BR/NPCs/Shared/MSS_VO_BR_Boss_Stereo
inputs [
 "Delay Max (Seconds)" {type=Time}
 "Delay Min (Seconds)" {type=Time}
 "Probability (0-1)" {type=Float}
 "Random Pitch Max (Semitones)" {type=Float}
 "Random Pitch Min (Semitones)" {type=Float}
 "Random Volume Change Max (Decibels)" {type=Float}
 "Random Volume Change Min (Decibels)" {type=Float}
 UE.Source.OnPlay {type=Trigger}
 Wave Array {type=WaveAsset:Array}
]
interfaces [UE.OutputFormat.Stereo:1.0, UE.Source.OneShot:1.0, UE.Source:1.1]
outputs [
 "UE.OutputFormat.Stereo.Audio:0" {type=Audio}
 "UE.OutputFormat.Stereo.Audio:1" {type=Audio}
 UE.Source.OneShot.OnFinished {type=Trigger}
]
pages [Default {input_defaults {
   "Delay Max (Seconds)"=1.75
   "Delay Min (Seconds)"=1.75
   "Probability (0-1)"=1
   "Random Pitch Max (Semitones)"=0
   "Random Pitch Min (Semitones)"=0
   "Random Volume Change Max (Decibels)"=5
   "Random Volume Change Min (Decibels)"=5
   UE.Source.OnPlay=null
   Wave Array [
    /JethroLager/Assets/Audio/VO/HeavyLamp/Waves/IP_VO_HeavyLamp_Death_030
    /JethroLager/Assets/Audio/VO/HeavyLamp/Waves/IP_VO_HeavyLamp_Death_040
    /JethroLager/Assets/Audio/VO/HeavyLamp/Waves/IP_VO_HeavyLamp_Death_050
   ]
  }}]
properties {
 assetImportData=&self:AssetImportData
 bEnableBaseSubmix=false
 numChannels=2
 outputFormat=Stereo
 sampleRate=48000
 soundClassObject=/Game/Sounds/FortSoundClassesAndModes/Fort_InGame_VOIP
 sourceEffectChain=/JethroLager/Assets/Audio/VO/HeavyLamp/SEC_IP_HeavyLamp
 volume=1.3
 concurrencyOverrides {bLimitToOwner=true, maxCount=1, resolutionRule=PreventNew}
 modulationSettings {
  volumeRouting=Override
  volumeModulationDestination {bEnableModulation=true, modulators [/JethroLager/Assets/Audio/VO/HeavyLamp/Modulation/CB_HeavyLamp_Voice]}
 }
 preEffectBusSends [{
   audioBus=/Game/Sounds/Modulation/EnvelopeFollowers/Volume/AB_Sidechain_NPCVoiceChat_Vol
   maxSendDistance=100
   minSendDistance=8000
   minSendLevel=1
   sendLevel=1
  }]
}
