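# One-shot mono MetaSound Source that plays HeavyLamp defeat voice lines, using the critical NPC VO attenuation and enemy-critical sound class.
# It wraps the shared MSS_VO_BR_Boss graph, feeding it four defeat waves with a fixed volume boost and no delay or pitch variation.
# A HeavyLamp effect chain colors the voice, with sends to distant reverb and music, weapon and explosion sidechain buses.
# Boss combat-chatter concurrency applies.
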MetaSoundSource /JethroLager/Assets/Audio/VO/HeavyLamp/MSS/MSS_IP_VO_HeavyLamp_Defeat
inputs [
 "Delay Max (Seconds)" {type=Time}
 "Delay Min (Seconds)" {type=Time}
 "Probability (0-1)" {type=Float}
 "Random Pitch Max (Semitones)" {type=Float}
 "Random Pitch Min (Semitones)" {type=Float}
 "Random Volume Change Max (Decibels)" {type=Float}
 "Random Volume Change Min (Decibels)" {type=Float}
 UE.Source.OnPlay {type=Trigger}
 Wave Array {type=WaveAsset:Array}
]
interfaces [UE.OutputFormat.Mono:1.0, UE.Source.OneShot:1.0, UE.Source:1.1]
outputs [
 "UE.OutputFormat.Mono.Audio:0" {type=Audio}
 UE.Source.OneShot.OnFinished {type=Trigger}
]
pages [Default {
  input_defaults {
   "Delay Max (Seconds)"=0
   "Delay Min (Seconds)"=0
   "Probability (0-1)"=1
   "Random Pitch Max (Semitones)"=0
   "Random Pitch Min (Semitones)"=0
   "Random Volume Change Max (Decibels)"=6
   "Random Volume Change Min (Decibels)"=6
   UE.Source.OnPlay=null
   Wave Array [
    /JethroLager/Assets/Audio/VO/HeavyLamp/Waves/IP_VO_HeavyLamp_Defeat_030
    /JethroLager/Assets/Audio/VO/HeavyLamp/Waves/IP_VO_HeavyLamp_Defeat_040
    /JethroLager/Assets/Audio/VO/HeavyLamp/Waves/IP_VO_HeavyLamp_Defeat_050
    /JethroLager/Assets/Audio/VO/HeavyLamp/Waves/IP_VO_HeavyLamp_Defeat_060
   ]
  }
  nodes [MSS_VO_BR_Boss_0 = /Game/Sounds/Fort_Human_VO_BR/NPCs/Shared/MSS_VO_BR_Boss(
    "Delay Max (Seconds)"=^"Delay Max (Seconds)"
    "Delay Min (Seconds)"=^"Delay Min (Seconds)"
    "Probability (0-1)"=^"Probability (0-1)"
    "Random Pitch Max (Semitones)"=^"Random Pitch Max (Semitones)"
    "Random Pitch Min (Semitones)"=^"Random Pitch Min (Semitones)"
    "Random Volume Change Max (Decibels)"=^"Random Volume Change Max (Decibels)"
    "Random Volume Change Min (Decibels)"=^"Random Volume Change Min (Decibels)"
    UE.Source.OnPlay=^UE.Source.OnPlay
    Wave Array=^Wave Array
   )]
  links [
   {dst="@UE.OutputFormat.Mono.Audio:0", src=MSS_VO_BR_Boss_0.UE.OutputFormat.Mono.Audio:0}
   {dst="@UE.Source.OneShot.OnFinished", src=MSS_VO_BR_Boss_0.UE.Source.OneShot.OnFinished}
  ]
 }]
properties {
 assetImportData=&self:AssetImportData
 attenuationSettings=/Game/Sounds/Fort_Human_VO_BR/NPCs/Shared/ATT_NPC_VO_Critical
 maxDistance=15000
 sampleRate=48000
 soundClassObject=&Fort_VO_Enemy_Critical
 sourceEffectChain=/JethroLager/Assets/Audio/VO/HeavyLamp/SEC_IP_HeavyLamp
 volume=3
 busSends [
  {audioBus=&AB_Sidechain_FortMusic_Vol, sendLevel=0.75}
  {audioBus=/Game/Sounds/Modulation/EnvelopeFollowers/Volume/AB_Sidechain_FortWeapon1P_Vol, sendLevel=0.75}
  {audioBus=&AB_Sidechain_FortWeapon_Vol, sendLevel=0.5}
  {audioBus=&AB_Sidechain_FortExplosion_Vol, sendLevel=0.5}
 ]
 concurrencyOverrides {bLimitToOwner=true, maxCount=1, resolutionRule=PreventNew}
 concurrencySet [&NPC_Boss_Combat_Chatter_Conc]
 modulationSettings {volumeModulationDestination {bEnableModulation=true}}
 soundSubmixSends [{
   maxSendDistance=8000
   maxSendLevel=0.5
   minSendDistance=2000
   minSendLevel=0.05
   sendLevel=0.1
   sendLevelControlMethod=Linear
   soundSubmix=&Fortnite_Reverb_Distant
  }]
}
